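# Physics asset for the HazMat backpack cosmetic, defining simulated collision for its hanging blaster and cable.
# Three capsule bodies cover the root, dyn_blaster and dyn_cable bones.
# Constraints form a chain: the blaster joins the root with both swing axes locked, and the cable joins the blaster with only swing1 locked.
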PhysicsAsset /BRCosmetics/Accessories/FORT_Backpacks/HazMat_Backpack/Meshes/SK_HazMat_Backpack_Physics
bodies {
 dyn_blaster [{shapeType=Capsule, length=23.1844, radius=5.2662, center=<-2.8675 0.6983 -11.5412>, rotation=<rot -2.2207 -90.2802 -0.7506>}]
 dyn_cable [{shapeType=Capsule, length=1.8937, radius=1.9325, center=<-0.522 -1.2779 -8.063>, rotation=<rot -0.8433 -92.3176 -100.8726>}]
 root [{shapeType=Capsule, length=27.1712, radius=12.2829, center=<5.3696 -9.7723 -2.964>}]
}
constraints [
 {
  bone1Name=dyn_blaster
  bone2Name=root
  swing1LimitDegrees=45
  swing1Motion=Locked
  swing2LimitDegrees=45
  swing2Motion=Locked
  twistLimitDegrees=10
 }
 {bone1Name=dyn_cable, bone2Name=dyn_blaster, swing1LimitDegrees=90, swing1Motion=Locked, swing2LimitDegrees=90, twistLimitDegrees=45}
]
